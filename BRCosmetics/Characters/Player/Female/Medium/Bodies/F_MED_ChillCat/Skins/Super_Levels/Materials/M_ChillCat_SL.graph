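# Character material for the Super Levels skin of ChillCat, using material attributes with shading model set by expression.
# Base PreFX and CharacterSkinFX feed three chained Season 22 Super Levels tier functions (T1–T3), which share an SL_FX mask parameter and AO.
# The tier functions and default mask texture come from the Bites skin.
# The T3 result branches into PostFX with a cinematic post pass, and into a ChillCat eyelash function.
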Material /BRCosmetics/Characters/Player/Female/Medium/Bodies/F_MED_ChillCat/Skins/Super_Levels/Materials/M_ChillCat_SL
graph {
 properties {
  bAutomaticallySetUsageInEditor=false
  bUseMaterialAttributes=true
  bUsedWithMeshDeformer=true
  bUsedWithMorphTargets=true
  bUsedWithSkeletalMesh=true
  pixelDepthOffsetMode=PDOM_Legacy
  refractionMethod=RM_IndexOfRefraction
  shadingModel=MSM_FromMaterialExpression
  subsurfaceProfile=&SS_HeroSkin_02
  usedShadingModels=Default Lit | Subsurface | Clear Coat | Subsurface Profile | Cloth | Eye
 }
 nodes [
  n1 = MaterialFunctionCall() {materialFunction=&BaseFN_CharacterFunction_PreFX}
  n2 = MaterialFunctionCall(
   PreskinnedPosition=@n1.Pre skinned Local Position
   Gmap=@n1.Pre skinned Local Position
   MaterialAttributes=@n1.Pre skinned Local Position
   AO=@n1.Pre skinned Local Position
  ) {materialFunction=&CharacterSkinFX}
  n3 = TextureSampleParameter2D() {
   group=S22_SL
   parameterName=SL_FX
   samplerType=SAMPLERTYPE_Masks
   sortPriority=1
   texture=/BRCosmetics/Characters/Player/Male/Medium/Bodies/M_MED_Bites/Skins/Super_Levels/Textures/T_Character_Bites_Body_SL_FX
  }
  n4 = MaterialFunctionCall!(
   MA=MaterialFunctionCall!(
    MA=MaterialFunctionCall!(MA=@n2.Result, SL_FX=@n3.RGBA, AO=@n1.AO) {materialFunction=/BRCosmetics/Characters/Player/Male/Medium/Bodies/M_MED_Bites/Skins/Super_Levels/Materials/MF_SL_S22_T1}.Result
    SL_FX=@n3.RGBA
    AO=@n1.AO
    PreSkinnedLocalPosition_VS
    PreSkinnedLocalNormal_PS
   ) {materialFunction=/BRCosmetics/Characters/Player/Male/Medium/Bodies/M_MED_Bites/Skins/Super_Levels/Materials/MF_SL_S22_T2}.Result
   SL_FX=@n3.RGBA
   AO=@n1.AO
  ) {materialFunction=/BRCosmetics/Characters/Player/Male/Medium/Bodies/M_MED_Bites/Skins/Super_Levels/Materials/MF_SL_S22_T3}
  n5 = MaterialFunctionCall(AO=@n1.AO, Distance Blend=@n1.AO, Fresnel=@n1.AO, Gmap=@n1.AO, Pre Skinned Local Position=@n1.AO, Material=@n4.Result) {materialFunction=&BaseFN_CharacterFunction_PostFX}
  MaterialFunctionCall(In=@n5.Result) {materialFunction=&CINE_MF_CharPostFX}
  MaterialFunctionCall!(In=@n4.Result) {materialFunction=/BRCosmetics/Characters/Player/Female/Medium/Bodies/F_MED_ChillCat/Materials/MF_ChillCat_Eyelash}
 ]
}
